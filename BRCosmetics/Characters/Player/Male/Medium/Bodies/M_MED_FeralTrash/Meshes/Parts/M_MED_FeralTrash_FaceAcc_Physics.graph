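# Physics asset for the face accessory mesh of the medium male FeralTrash cosmetic body, defining bodies for three hair chains (F, G, H) and a hook.
# Each chain starts at a tiny master capsule constrained to a large, zero-length head capsule with free swing and twist.
# Downstream capsules have small swing limits and zero twist limit, with the hook looser than the hair.
# Hair F's third segment constrains to its first segment rather than its second.
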PhysicsAsset /BRCosmetics/Characters/Player/Male/Medium/Bodies/M_MED_FeralTrash/Meshes/Parts/M_MED_FeralTrash_FaceAcc_Physics
bodies {
 dyn_hairF_01 [{shapeType=Capsule, length=2, radius=2, shapeName=dyn_hairF_01_capsule, center=<1.6286 0 0>, rotation=<rot 90 0 0>}]
 dyn_hairF_02 [{shapeType=Capsule, length=2, radius=2, shapeName=dyn_hairF_02_capsule, center=<1.6286 0 0>, rotation=<rot 90 0 0>}]
 dyn_hairF_03 [{shapeType=Capsule, length=2, radius=2, shapeName=dyn_hairF_03_capsule, center=<1.6286 0 0>, rotation=<rot 90 0 0>}]
 dyn_hairF_MSTR [{shapeType=Capsule, length=0.505, radius=0.505, shapeName=dyn_hairF_MSTR_capsule}]
 dyn_hairG_01 [{shapeType=Capsule, length=2, radius=2, shapeName=dyn_hairG_01_capsule, center=<1.6286 0 0>, rotation=<rot 90 0 0>}]
 dyn_hairG_02 [{shapeType=Capsule, length=2, radius=2, shapeName=dyn_hairG_02_capsule, center=<1.6286 0 0>, rotation=<rot 90 0 0>}]
 dyn_hairG_MSTR [{shapeType=Capsule, length=0.505, radius=0.505, shapeName=dyn_hairG_MSTR_capsule}]
 dyn_hairH_01 [{shapeType=Capsule, length=2, radius=2, shapeName=dyn_hairH_01_capsule, center=<1.6286 0 0>, rotation=<rot 90 0 0>}]
 dyn_hairH_02 [{shapeType=Capsule, length=2, radius=2, shapeName=dyn_hairH_02_capsule, center=<1.6286 0 0>, rotation=<rot 90 0 0>}]
 dyn_hairH_MSTR [{shapeType=Capsule, length=0.505, radius=0.505, shapeName=dyn_hairH_MSTR_capsule}]
 dyn_hook_01 [{shapeType=Capsule, length=2.2086, radius=1.4777, shapeName=dyn_hook_01_capsule, center=<1.8945 0 0>, rotation=<rot 90 0 0>}]
 dyn_hook_02 [{shapeType=Capsule, length=3.4405, radius=2.0613, shapeName=dyn_hook_02_capsule, center=<3.772 0 0>, rotation=<rot 90 0 0>}]
 dyn_hook_MSTR [{shapeType=Capsule, length=0.1, radius=0.1, shapeName=dyn_hook_MSTR_capsule}]
 head [{shapeType=Capsule, length=0, radius=11.5, center=<4.7465 3.89 0>}]
}
constraints [
 {bone1Name=dyn_hairH_02, bone2Name=dyn_hairH_01, swing1LimitDegrees=10, swing2LimitDegrees=10, twistLimitDegrees=0}
 {bone1Name=dyn_hairF_01, bone2Name=dyn_hairF_MSTR, swing1LimitDegrees=5, swing2LimitDegrees=5, twistLimitDegrees=0}
 {bone1Name=dyn_hairF_02, bone2Name=dyn_hairF_01, swing1LimitDegrees=5, swing2LimitDegrees=5, twistLimitDegrees=0}
 {bone1Name=dyn_hairF_03, bone2Name=dyn_hairF_01, swing1LimitDegrees=10, swing2LimitDegrees=10, twistLimitDegrees=0}
 {bone1Name=dyn_hairG_01, bone2Name=dyn_hairG_MSTR, swing1LimitDegrees=10, swing2LimitDegrees=10, twistLimitDegrees=0}
 {bone1Name=dyn_hairG_02, bone2Name=dyn_hairG_01, swing1LimitDegrees=10, swing2LimitDegrees=10, twistLimitDegrees=0}
 {bone1Name=dyn_hairH_01, bone2Name=dyn_hairH_MSTR, swing1LimitDegrees=10, swing2LimitDegrees=10, twistLimitDegrees=0}
 {
  bone1Name=dyn_hairF_MSTR
  bone2Name=head
  swing1LimitDegrees=45
  swing1Motion=Free
  swing2LimitDegrees=45
  swing2Motion=Free
  twistLimitDegrees=45
  twistMotion=Free
 }
 {
  bone1Name=dyn_hairG_MSTR
  bone2Name=head
  swing1LimitDegrees=45
  swing1Motion=Free
  swing2LimitDegrees=45
  swing2Motion=Free
  twistLimitDegrees=45
  twistMotion=Free
 }
 {
  bone1Name=dyn_hairH_MSTR
  bone2Name=head
  swing1LimitDegrees=45
  swing1Motion=Free
  swing2LimitDegrees=45
  swing2Motion=Free
  twistLimitDegrees=45
  twistMotion=Free
 }
 {
  bone1Name=dyn_hook_MSTR
  bone2Name=head
  swing1LimitDegrees=45
  swing1Motion=Free
  swing2LimitDegrees=45
  swing2Motion=Free
  twistLimitDegrees=45
  twistMotion=Free
 }
 {bone1Name=dyn_hook_01, bone2Name=dyn_hook_MSTR, swing1LimitDegrees=15, swing2LimitDegrees=15, twistLimitDegrees=0}
 {bone1Name=dyn_hook_02, bone2Name=dyn_hook_01, swing1LimitDegrees=15, swing2LimitDegrees=15, twistLimitDegrees=0}
]
