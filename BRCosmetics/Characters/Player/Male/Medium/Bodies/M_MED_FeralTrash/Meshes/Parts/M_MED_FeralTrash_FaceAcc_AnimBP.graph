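# Post-attach Animation Blueprint for the male FeralTrash body's face accessory mesh, adding secondary motion to its dynamic bones.
# It copies the parent mesh's pose, then blends in an additive trash can bone offset while bIsFalling is true.
# A head-relative RigidBody node then simulates the accessory physics asset with dampened component motion before output.
# Four chained hair AnimDynamics nodes are disconnected from the output, and the event graph has no logic.
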AnimBlueprint /BRCosmetics/Characters/Player/Male/Medium/Bodies/M_MED_FeralTrash/Meshes/Parts/M_MED_FeralTrash_FaceAcc_AnimBP
target_skeleton=&/M_MED_FeralTrash_FaceAcc_Skeleton
graphs [
 AnimGraph {nodes [
   n1 = Anim Dynamics - Bone: dyn_hairG_01(Alpha=1, ExternalForce="0,0,0") {
    angularDampingOverride=1
    angularSpringConstant=20000
    bAngularSpring=true
    bOverrideAngularDamping=true
    bUseGravityOverride=true
    gravityScale=0.35
    simulationSpace=RootRelative
    alphaBoolBlend {blendOption=Linear}
    boundBone {boneName=dyn_hairG_01}
    componentLinearAccScale=<1 1 1>
    componentLinearVelScale=<1 1 1>
    gravityOverride {z=-980}
    physicsBodyDefinitions [{
      boundBone {boneName=dyn_hairG_01}
      boxExtents=<20 30 30>
      constraintSetup {angularLimitsMax=<5 5 2>, angularLimitsMin=<-5 -5 -10>, angularTarget=<1 0 0>}
      localJointOffset=<4 0 0>
     }]
   }
   n2 = Anim Dynamics - Bone: dyn_hairH_01(ComponentPose=@n1, Alpha=1, ExternalForce="0,0,0") {
    angularDampingOverride=1
    angularSpringConstant=20000
    bAngularSpring=true
    bOverrideAngularDamping=true
    bUseGravityOverride=true
    gravityScale=0.35
    simulationSpace=RootRelative
    alphaBoolBlend {blendOption=Linear}
    boundBone {boneName=dyn_hairH_01}
    componentLinearAccScale=<1 1 1>
    componentLinearVelScale=<1 1 1>
    gravityOverride {z=-980}
    physicsBodyDefinitions [{
      boundBone {boneName=dyn_hairH_01}
      boxExtents=<20 30 30>
      constraintSetup {angularLimitsMax=<5 5 2>, angularLimitsMin=<-5 -5 -10>, angularTarget=<1 0 0>}
      localJointOffset=<4 0 0>
     }]
   }
   n3 = Anim Dynamics - Bone: dyn_hairG_02(ComponentPose=@n2, Alpha=1, ExternalForce="0,0,0") {
    angularDampingOverride=1
    angularSpringConstant=20000
    bAngularSpring=true
    bOverrideAngularDamping=true
    bUseGravityOverride=true
    gravityScale=0.35
    simulationSpace=RootRelative
    alphaBoolBlend {blendOption=Linear}
    boundBone {boneName=dyn_hairG_02}
    componentLinearAccScale=<1 1 1>
    componentLinearVelScale=<1 1 1>
    gravityOverride {z=-980}
    physicsBodyDefinitions [{
      boundBone {boneName=dyn_hairG_02}
      boxExtents=<20 30 30>
      constraintSetup {angularLimitsMax=<5 5 2>, angularLimitsMin=<-5 -5 -10>, angularTarget=<1 0 0>}
      localJointOffset=<4 0 0>
     }]
   }
   Anim Dynamics - Bone: dyn_hairH_02(ComponentPose=@n3, Alpha=1, ExternalForce="0,0,0") {
    angularDampingOverride=1
    angularSpringConstant=20000
    bAngularSpring=true
    bOverrideAngularDamping=true
    bUseGravityOverride=true
    gravityScale=0.35
    simulationSpace=RootRelative
    alphaBoolBlend {blendOption=Linear}
    boundBone {boneName=dyn_hairH_02}
    componentLinearAccScale=<1 1 1>
    componentLinearVelScale=<1 1 1>
    gravityOverride {z=-980}
    physicsBodyDefinitions [{
      boundBone {boneName=dyn_hairH_02}
      boxExtents=<20 30 30>
      constraintSetup {angularLimitsMax=<5 5 2>, angularLimitsMin=<-5 -5 -10>, angularTarget=<1 0 0>}
      localJointOffset=<4 0 0>
     }]
   }
   n4 = "Transform (Modify) Bone - Bone: dyn_trashCan_MSTR"!(
    ComponentPose=Local To Component(LocalPose=Copy Pose From Mesh() {bUseAttachedParent=true})
    bAlphaBoolEnabled:bool=Get bIsFalling()
    Translation="2.75,0,0"
    Rotation="2,2,1.743"
    Scale="1,1,1"
   ) -> (Pose) {
    alphaInputType=Bool
    rotationMode=BMM_Additive
    rotationSpace=BCS_BoneSpace
    translationMode=BMM_Additive
    translationSpace=BCS_BoneSpace
    alphaBoolBlend {blendInTime=0.35, blendOption=Linear, blendOutTime=0.2}
    boneToModify {boneName=dyn_trashCan_MSTR}
    rotation=<rot 2 2 1.743>
    translation {x=2.75}
   }
   n5 = RigidBody(ComponentPose=@n4) {
    bClampLinearTranslationLimitToRefPose=true
    bOverrideWorldGravity=true
    overridePhysicsAsset=&/M_MED_FeralTrash_FaceAcc_Physics
    simulationSpace=BaseBoneSpace
    alphaBoolBlend {blendOption=Linear}
    baseBoneRef {boneName=head}
    componentAppliedLinearAccClamp=<6000 6000 6000>
    componentLinearAccScale=<0.3 0.3 0.3>
    componentLinearVelScale=<0.3 0.3 0.3>
    overrideWorldGravity {z=-980}
   }
   Output Pose(Result=Component To Local(ComponentPose=@n5)) {layerGroup=None}
  ]}
 EventGraph {nodes [Event BlueprintUpdateAnimation()]}
]
variables []
