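# Physics asset for the hair and face accessory mesh of the male medium FeralTrash player body, giving its hair bones capsule bodies.
# Eight small master bodies, hairA to H, each constrain to one large round head capsule with free swing and twist.
# Chains A to E add two segments below their master, with narrow swing limits and zero twist limit, while F, G and H have only the master body.
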PhysicsAsset /BRCosmetics/Characters/Player/Male/Medium/Bodies/M_MED_FeralTrash/Meshes/Parts/M_MED_FeralTrash_Hair_FaceAcc_Physics
bodies {
 dyn_hairA_01 [{shapeType=Capsule, length=2, radius=2, shapeName=dyn_hairA_01_capsule, center=<1.6286 0 0>, rotation=<rot 90 0 0>}]
 dyn_hairA_02 [{shapeType=Capsule, length=2, radius=2, shapeName=dyn_hairA_02_capsule, center=<1.6286 0 0>, rotation=<rot 90 0 0>}]
 dyn_hairA_MSTR [{shapeType=Capsule, length=0.505, radius=0.505, shapeName=dyn_hairA_MSTR_capsule}]
 dyn_hairB_01 [{shapeType=Capsule, length=2, radius=2, shapeName=dyn_hairB_01_capsule, center=<1.6286 0 0>, rotation=<rot 90 0 0>}]
 dyn_hairB_02 [{shapeType=Capsule, length=2, radius=2, shapeName=dyn_hairB_02_capsule, center=<1.6286 0 0>, rotation=<rot 90 0 0>}]
 dyn_hairB_MSTR [{shapeType=Capsule, length=0.505, radius=0.505, shapeName=dyn_hairB_MSTR_capsule}]
 dyn_hairC_01 [{shapeType=Capsule, length=2, radius=2, shapeName=dyn_hairC_01_capsule, center=<1.6286 0 0>, rotation=<rot 90 0 0>}]
 dyn_hairC_02 [{shapeType=Capsule, length=2, radius=2, shapeName=dyn_hairC_02_capsule, center=<1.6286 0 0>, rotation=<rot 90 0 0>}]
 dyn_hairC_MSTR [{shapeType=Capsule, length=0.505, radius=0.505, shapeName=dyn_hairC_MSTR_capsule}]
 dyn_hairD_01 [{shapeType=Capsule, length=2, radius=2, shapeName=dyn_hairD_01_capsule, center=<1.6286 0 0>, rotation=<rot 90 0 0>}]
 dyn_hairD_02 [{shapeType=Capsule, length=2, radius=2, shapeName=dyn_hairD_02_capsule, center=<1.6286 0 0>, rotation=<rot 90 0 0>}]
 dyn_hairD_MSTR [{shapeType=Capsule, length=0.505, radius=0.505, shapeName=dyn_hairD_MSTR_capsule}]
 dyn_hairE_01 [{shapeType=Capsule, length=2, radius=2, shapeName=dyn_hairE_01_capsule, center=<1.6286 0 0>, rotation=<rot 90 0 0>}]
 dyn_hairE_02 [{shapeType=Capsule, length=2, radius=2, shapeName=dyn_hairE_02_capsule, center=<1.6286 0 0>, rotation=<rot 90 0 0>}]
 dyn_hairE_MSTR [{shapeType=Capsule, length=0.505, radius=0.505, shapeName=dyn_hairE_MSTR_capsule}]
 dyn_hairF_MSTR [{shapeType=Capsule, length=0.505, radius=0.505, shapeName=dyn_hairF_MSTR_capsule}]
 dyn_hairG_MSTR [{shapeType=Capsule, length=0.505, radius=0.505, shapeName=dyn_hairG_MSTR_capsule}]
 dyn_hairH_MSTR [{shapeType=Capsule, length=0.505, radius=0.505, shapeName=dyn_hairH_MSTR_capsule}]
 head [{shapeType=Capsule, length=0, radius=10.2198, center=<4.4277 0.8187 0.2937>, rotation=<rot 0 130.8705 -88.3104>}]
}
constraints [
 {bone1Name=dyn_hairA_01, bone2Name=dyn_hairA_MSTR, swing1LimitDegrees=20, swing2LimitDegrees=15, twistLimitDegrees=0}
 {bone1Name=dyn_hairA_02, bone2Name=dyn_hairA_01, swing1LimitDegrees=10, swing2LimitDegrees=15, twistLimitDegrees=0}
 {bone1Name=dyn_hairB_01, bone2Name=dyn_hairB_MSTR, swing1LimitDegrees=10, swing2LimitDegrees=15, twistLimitDegrees=0}
 {bone1Name=dyn_hairB_02, bone2Name=dyn_hairB_01, swing1LimitDegrees=20, swing2LimitDegrees=15, twistLimitDegrees=0}
 {bone1Name=dyn_hairC_01, bone2Name=dyn_hairC_MSTR, swing1LimitDegrees=10, swing2LimitDegrees=15, twistLimitDegrees=0}
 {bone1Name=dyn_hairC_02, bone2Name=dyn_hairC_01, swing1LimitDegrees=20, swing2LimitDegrees=15, twistLimitDegrees=0}
 {bone1Name=dyn_hairD_01, bone2Name=dyn_hairD_MSTR, swing1LimitDegrees=20, swing2LimitDegrees=15, twistLimitDegrees=0}
 {bone1Name=dyn_hairD_02, bone2Name=dyn_hairD_01, swing1LimitDegrees=20, swing2LimitDegrees=15, twistLimitDegrees=0}
 {bone1Name=dyn_hairE_01, bone2Name=dyn_hairE_MSTR, swing1LimitDegrees=20, swing2LimitDegrees=15, twistLimitDegrees=0}
 {bone1Name=dyn_hairE_02, bone2Name=dyn_hairE_01, swing1LimitDegrees=20, swing2LimitDegrees=15, twistLimitDegrees=0}
 {
  bone1Name=dyn_hairA_MSTR
  bone2Name=head
  swing1LimitDegrees=45
  swing1Motion=Free
  swing2LimitDegrees=45
  swing2Motion=Free
  twistLimitDegrees=45
  twistMotion=Free
 }
 {
  bone1Name=dyn_hairB_MSTR
  bone2Name=head
  swing1LimitDegrees=45
  swing1Motion=Free
  swing2LimitDegrees=45
  swing2Motion=Free
  twistLimitDegrees=45
  twistMotion=Free
 }
 {
  bone1Name=dyn_hairC_MSTR
  bone2Name=head
  swing1LimitDegrees=45
  swing1Motion=Free
  swing2LimitDegrees=45
  swing2Motion=Free
  twistLimitDegrees=45
  twistMotion=Free
 }
 {
  bone1Name=dyn_hairD_MSTR
  bone2Name=head
  swing1LimitDegrees=45
  swing1Motion=Free
  swing2LimitDegrees=45
  swing2Motion=Free
  twistLimitDegrees=45
  twistMotion=Free
 }
 {
  bone1Name=dyn_hairE_MSTR
  bone2Name=head
  swing1LimitDegrees=45
  swing1Motion=Free
  swing2LimitDegrees=45
  swing2Motion=Free
  twistLimitDegrees=45
  twistMotion=Free
 }
 {
  bone1Name=dyn_hairF_MSTR
  bone2Name=head
  swing1LimitDegrees=45
  swing1Motion=Free
  swing2LimitDegrees=45
  swing2Motion=Free
  twistLimitDegrees=45
  twistMotion=Free
 }
 {
  bone1Name=dyn_hairG_MSTR
  bone2Name=head
  swing1LimitDegrees=45
  swing1Motion=Free
  swing2LimitDegrees=45
  swing2Motion=Free
  twistLimitDegrees=45
  twistMotion=Free
 }
 {
  bone1Name=dyn_hairH_MSTR
  bone2Name=head
  swing1LimitDegrees=45
  swing1Motion=Free
  swing2LimitDegrees=45
  swing2Motion=Free
  twistLimitDegrees=45
  twistMotion=Free
 }
]
